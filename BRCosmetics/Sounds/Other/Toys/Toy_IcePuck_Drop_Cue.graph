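# Sound cue for an ice puck toy being dropped, using its landing sounds.
# A Random node picks one of three landing waves with equal weight.
# It borrows the weapon drop attenuation and is routed to the third-person enemy player foley sound class.
# Volume is boosted, and a concurrency override limits voices per owner.
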SoundCue /BRCosmetics/Sounds/Other/Toys/Toy_IcePuck_Drop_Cue
root=Random_0
nodes [
 {id=Random_0, children [WavePlayer_0,WavePlayer_1,WavePlayer_2], properties {weights [1,1,1]}}
 {id=WavePlayer_0, properties {soundWaveAssetPtr=/Game/Athena/Sounds/Toys/Toy_IcePuck_Land_01}}
 {id=WavePlayer_1, properties {soundWaveAssetPtr=/Game/Athena/Sounds/Toys/Toy_IcePuck_Land_03}}
 {id=WavePlayer_2, properties {soundWaveAssetPtr=/Game/Athena/Sounds/Toys/Toy_IcePuck_Land_02}}
]
properties {
 attenuationSettings=&Fort_Weapon_Drop_Att
 maxDistance=3000
 soundClassObject=&Fort_Player_Foley_3P_Enemy_Athena
 volumeMultiplier=1.5873
 concurrencyOverrides {bLimitToOwner=true}
}
