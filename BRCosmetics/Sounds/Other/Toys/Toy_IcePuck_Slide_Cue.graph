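# Sound cue for the sliding loop of the ice puck toy.
# A looping wave player feeds a smoothed Doppler node, so pitch shifts with relative motion between puck and listener.
# It uses the shared toy-hit attenuation and the third-person enemy player foley sound class, with a slight volume boost and concurrency limited per owner.
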SoundCue /BRCosmetics/Sounds/Other/Toys/Toy_IcePuck_Slide_Cue
root=Doppler_0
nodes [
 {id=Doppler_0, children [WavePlayer_7], properties {bUseSmoothing=true}}
 {id=WavePlayer_7, properties {bLooping=true, soundWaveAssetPtr=&/Toy_IcePuck_Slide}}
]
properties {
 attenuationSettings=/Game/Sounds/Attenuation/Fort_ToyHit_Att
 maxDistance=2800
 soundClassObject=&Fort_Player_Foley_3P_Enemy_Athena
 volumeMultiplier=1.1905
 concurrencyOverrides {bLimitToOwner=true}
}
