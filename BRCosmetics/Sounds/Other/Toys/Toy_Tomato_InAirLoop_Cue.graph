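# Sound cue for the looping in-air sound of the Tomato toy, played as a quiet gadget-class sound with toy-hit attenuation.
# A looping wave feeds a Modulator that randomizes pitch upward, then a smoothed, exaggerated Doppler node that shifts pitch with relative motion.
# An Enveloper at the root briefly ramps volume from half to full at the start.
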SoundCue /BRCosmetics/Sounds/Other/Toys/Toy_Tomato_InAirLoop_Cue
root=Enveloper_0
nodes [
 {id=Doppler_0, children [Modulator_0], properties {bUseSmoothing=true, dopplerIntensity=3, smoothingInterpSpeed=0.3}}
 {
  id=Enveloper_0
  children [Doppler_0]
  properties {pitchCurve {editorCurveData {keys [{time=0}]}}, volumeCurve {editorCurveData {keys [<at 0 0.5>,<at 0.2 1>]}}}
 }
 {id=Modulator_0, children [WavePlayer_0], properties {pitchMax=1.5, pitchMin=1}}
 {id=WavePlayer_0, properties {bLooping=true, soundWaveAssetPtr=/Game/Athena/Sounds/Toys/Toy_Tomato_InAirLoop}}
]
properties {
 attenuationSettings=/Game/Sounds/Attenuation/Fort_ToyHit_Att
 maxDistance=2800
 soundClassObject=&Fort_Gadgets
 volumeMultiplier=0.3
}
