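# Sound cue for the basketball toy's shoot sound.
# A random node picks between two shoot wave variations with equal weight.
# It uses the emote attenuation and emote music sound class, with a slightly boosted volume and limited audible range.
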SoundCue /BRCosmetics/Sounds/Other/Toys/Toys_BasketBall_Shoot_Cue
root=Random_1
nodes [
 {id=Random_1, children [WavePlayer_2,WavePlayer_3], properties {weights [1,1]}}
 {id=WavePlayer_2, properties {soundWaveAssetPtr=&/Toys_BasketBall_Shoot_07}}
 {id=WavePlayer_3, properties {soundWaveAssetPtr=&/Toys_BasketBall_Shoot_08}}
]
properties {
 attenuationSettings=&Fort_Emotes_Athena_Att
 maxDistance=1600
 soundClassObject=&Fort_Music_Emote
 volumeMultiplier=1.25
}
