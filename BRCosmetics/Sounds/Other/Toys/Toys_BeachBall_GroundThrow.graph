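# Sound cue for the beach ball toy's ground throw sound in the cosmetic toys set.
# A Random node picks between two wave variations with equal weight.
# It plays slightly boosted in volume, using Athena emote attenuation and the emote music sound class.
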SoundCue /BRCosmetics/Sounds/Other/Toys/Toys_BeachBall_GroundThrow
root=Random_3
nodes [
 {id=Random_3, children [WavePlayer_8,WavePlayer_9], properties {weights [1,1]}}
 {id=WavePlayer_8, properties {soundWaveAssetPtr=&/Toys_BeachBall_GroundThrow_03}}
 {id=WavePlayer_9, properties {soundWaveAssetPtr=&/Toys_BeachBall_GroundThrow_04}}
]
properties {
 attenuationSettings=&Fort_Emotes_Athena_Att
 maxDistance=1600
 soundClassObject=&Fort_Music_Emote
 volumeMultiplier=1.25
}
