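# Sound cue for the putt sound of a golf ball toy cosmetic.
# A single wave player feeds one putt recording into a modulator left at its usual settings, which is the root of the cue.
# The cue raises overall volume slightly, uses the shared emote attenuation and routes through the emote music sound class.
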SoundCue /BRCosmetics/Sounds/Other/Toys/Toys_GolfBall_Putt_Cue
root=Modulator_0
nodes [
 {id=Modulator_0, children [WavePlayer_0]}
 {id=WavePlayer_0, properties {soundWaveAssetPtr=&/Toys_GolfBall_Putt_11}}
]
properties {
 attenuationSettings=&Fort_Emotes_Athena_Att
 maxDistance=1600
 soundClassObject=&Fort_Music_Emote
 volumeMultiplier=1.25
}
